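; Gameplay ability for a BattlePass S26 quest that spawns gym equipment quest actors in response to a fishing spawn event.
; On activation it waits for that gameplay event, then uses a counter to pick among five snapshot classes, cycling so the player sees every piece.
; A helper function spawns drop-entry actors from the event target, launches them with generated velocity, checks cosmetics and optionally copies target tags.
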(event Ability|EventActivateAbility
  (bind _gym_equipment_counter (Variables|Default|GetGym%20Equipment%20Counter))
  (bind (_asynctaskproxy _payload) (Ability|Tasks|WaitGameplayEvent "(TagName=\"Event.Quest.SpawnQuestActor.Fishing\")")
    (:EventReceived
      ;[1 1 1] Make sure the player sees all the gym equipment pieces at least once.
      (if (<= _gym_equipment_counter 0)
        (Variables|Default|SetDrop%20Entry%20Spawned%20Quest%20Actor%20Class "/BattlePassS26_Quests/Items/QuestActors/Snapshots/Part03/BP_S26_Snapshot_03_Gym_01.BP_S26_Snapshot_03_Gym_01_C")
        (elif (== _gym_equipment_counter 1)
          (Variables|Default|SetDrop%20Entry%20Spawned%20Quest%20Actor%20Class "/BattlePassS26_Quests/Items/QuestActors/Snapshots/Part03/BP_S26_Snapshot_03_Gym_02.BP_S26_Snapshot_03_Gym_02_C")
          (elif (== _gym_equipment_counter 2)
            (Variables|Default|SetDrop%20Entry%20Spawned%20Quest%20Actor%20Class "/BattlePassS26_Quests/Items/QuestActors/Snapshots/Part03/BP_S26_Snapshot_03_Gym_03.BP_S26_Snapshot_03_Gym_03_C")
            (elif (== _gym_equipment_counter 3)
              (Variables|Default|SetDrop%20Entry%20Spawned%20Quest%20Actor%20Class "/BattlePassS26_Quests/Items/QuestActors/Snapshots/Part03/BP_S26_Snapshot_03_Gym_04.BP_S26_Snapshot_03_Gym_04_C")
              (elif (== _gym_equipment_counter 4)
                (Variables|Default|SetDrop%20Entry%20Spawned%20Quest%20Actor%20Class "/BattlePassS26_Quests/Items/QuestActors/Snapshots/Part03/BP_S26_Snapshot_03_Gym_05.BP_S26_Snapshot_03_Gym_05_C")
                ;[1 1 1] Increment the counter, but reset it and start counting again if they haven't gotten all the pieces for whatever reason.
                (Variables|Default|SetGym%20Equipment%20Counter 0)
                (bind (_payload_eventtag _payload_instigator _payload_target _payload_optionalobject _payload_optionalobject2 _payload_contexthandle _payload_instigatortags _payload_targettags _payload_eventmagnitude _payload_targetdata) (Utilities|Struct|BreakGameplayEventData _payload))
                (Spawning|SpawnQuestActorsFromDropEntry 0 (Utilities|Struct|MakeStructSpawnedQuestActorDropEntry (Variables|Default|GetDrop%20Entry%20Spawned%20Quest%20Actor%20Class) 1 1) (Utilities|Struct|MakeGameplayEventData _payload_eventtag _payload_instigator _payload_target _payload_optionalobject _payload_optionalobject2 _payload_contexthandle _payload_instigatortags _payload_targettags _payload_eventmagnitude _payload_targetdata))
                (else
                  _))))))
      (bind _result (Math|Integer|IncrementInt (Variables|Default|GetGym%20Equipment%20Counter))
        (:%20%20
          (Variables|Default|SetGym%20Equipment%20Counter _result))))))

(fn SpawnQuestActorsFromDropEntry
  (params (DropEntry :type Struct%20Spawned%20Quest%20Actor%20Drop%20Entry%20Structure) (EventData :type Gameplay%20Event%20Data%20Structure))
  (bind (_spawnedquestactorclass_6_f15cfde2424412e9badbe2975b317a07 _dropquantity_14_6469e6844f83772bb80669bcf263cb9d) (Utilities|Struct|BreakStructSpawnedQuestActorDropEntry DropEntry))
  (bind (_eventtag _instigator _target _optionalobject _optionalobject2 _contexthandle _instigatortags _targettags _eventmagnitude _targetdata) (Utilities|Struct|BreakGameplayEventData EventData))
  (bind (_eventtag_2 _instigator_2 _target_2 _optionalobject_2) (Utilities|Struct|BreakGameplayEventData EventData))
  (for _ (range 1 (+ _dropquantity_14_6469e6844f83772bb80669bcf263cb9d 1))
    (bind _returnvalue (Building|SpawnBuildingGameplayActor _spawnedquestactorclass_6_f15cfde2424412e9badbe2975b317a07 (Spawning|GetSpawnTransform _target _targetdata) _instigator))
    (bind _asbp_spawned_quest_actor_parent (Utilities|Casting|CastToBP_SpawnedQuestActor_Parent _returnvalue))
    (Variables|/Game/Athena/Items/QuestParents/SpawnedQuestActor/BP_SpawnedQuestActor_Parent.BP_SpawnedQuestActor_Parent_C|SetbTestVisualsNoQuestRequired false _asbp_spawned_quest_actor_parent)
    (Projectile|SetVelocity (Variables|/Game/Athena/Items/QuestParents/SpawnedQuestActor/BP_SpawnedQuestActor_Parent.BP_SpawnedQuestActor_Parent_C|GetFortProjectileMovement _asbp_spawned_quest_actor_parent) (Spawning|GenerateSpawnVelocity))
    (Class|BPSpawnedQuestActorParent|CheckCosmetics _asbp_spawned_quest_actor_parent (Fortnite|GetFortPlayerControllerfromActor (Variables|GAVars|Data|GetOwningPlayerPawn)))
    (if (Variables|Setup|GetSpawnedActorInheritsTags)
      (Ability|GameplayEffect|AddLooseGameplayTags _asbp_spawned_quest_actor_parent _targettags))
    (Utilities|IsValid _optionalobject_2
      (:Is%20Valid
        (bind _asactor (Utilities|Casting|CastToActor _optionalobject_2))))))
